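# Niagara system for the PopBucket pickaxe's motion trail, drawn as short-lived ribbons.
# Its single emitter, Main_Trails, uses AnimTrail modules to spawn, initialize and update particles from bone or socket positions on a system-level skeletal mesh, feeding two ribbon renderers.
# Time-of-day functions tint and dim the trail, shifting between pink, red and purple, while float curves fade both particle alpha and the motion-trail color's alpha.
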NiagaraSystem /BRCosmetics/Weapons/FORT_Melee/Pickaxe_PopBucket/FX/NS_Pickaxe_PopBucket_Trail
dependencies {
 usedDataInterfaces [&NiagaraDataInterfaceCurve, &NiagaraDataInterfaceSkeletalMesh]
 usedDynamicInputs []
 usedModules [
  &self:SystemSpawnScript.SystemScriptSource.SystemScriptGraph.NiagaraNodeAssignment_2.SetVariables_#1
  &SystemState
 ]
}
emitters {Main_Trails {
  emitterUpdateScript [
   &EmitterState(Life Cycle Mode=::NewEnumerator1, Loop Behavior=::NewEnumerator1, Loop Delay=0.05, Loop Duration=0.2, UseLoopDelay=false)
   &FN_SpawnAnimTrail_v2()
  ]
  particleSpawnScript [
   /Game/Effects/Niagara/Modules/AnimTrail/FN_InitializeAnimTrail/FN_InitializeAnimTrailBoneIndecies(First Bone or Socket Index=2, Second Bone or Socket Index=2)
   &FN_InitializeAnimTrail(AnimTrailWidth=32, Lifetime=0.08, SkeletalMesh=$System.Skeletal Mesh) {module=FN_InitializeAnimTrail001}
  ]
  particleUpdateScript [
   &FN_UpdateAnimTrail(SkeletalMesh=$System.Skeletal Mesh)
   &ScaleColor(
    Scale Alpha=&FloatFromCurve(FloatCurve=DataInterface() {propertyValues {Curve {keys [<at 0.2 1 RCIM_Cubic>, {tangentMode=RCTM_User, time=0.5, value=0}]}}})
    Scale RGB=&VectorFromFloat(Value=&FNTimeOfDayScalar(Night Scale=0.6, Morning Scale=1, Evening Scale=1))
    ScaleA=true
    ScaleRGB=true
   ) {module=ScaleColor002}
   /SetVariables(Particles.MotionTrailColor:LinearColor=&MultiplyLinearColor(
     Linear Color=&FNTimeOfDayColor(
      Night Color=<rgba 0.3629 0.1025 0.4948 1>
      Morning Color=<rgba 0.7917 0.0249 0.0119 1>
      Day Color=<rgba 1 0.0792 0.584 1>
      Evening Color=<rgba 0.7917 0.0249 0.0119 1>
     )
     Scale Factor=&MakeVector4(
      X=&FNTimeOfDayScalar(Night Scale=0.4, Morning Scale=0.5, Day Scale=0.5, Evening Scale=0.5)
      Y=&FNTimeOfDayScalar(Night Scale=0.4, Morning Scale=0.5, Day Scale=0.5, Evening Scale=0.5)
      Z=&FNTimeOfDayScalar(Night Scale=0.4, Morning Scale=0.5, Day Scale=0.5, Evening Scale=0.5)
      W=&FloatFromCurve(FloatCurve=DataInterface() {propertyValues {Curve {keys [<at 0 1 RCIM_Cubic>, {tangentMode=RCTM_User, time=0.35, value=0}]}}})
     )
    )) {isSetParameters=true}
  ]
  rendererClasses [&NiagaraRibbonRendererProperties]
  renderers [
   {rendererClass=&NiagaraRibbonRendererProperties}
   {rendererClass=&NiagaraRibbonRendererProperties, rendererIndex=1}
  ]
 }}
summary {
 systemName=NS_Pickaxe_PopBucket_Trail
 emitters [{emitterName=Main_Trails, rendererClasses [&NiagaraRibbonRendererProperties]}]
 userVariables [User.RibbonIDOffset {type=NiagaraInt32}]
}
